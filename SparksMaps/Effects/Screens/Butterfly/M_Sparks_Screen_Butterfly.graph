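# Unlit Sparks screen material showing an animated butterfly amid radial energy, shards and edge glow.
# A sine wave oscillates the butterfly's horizontal UV scale like flapping wings, with a second offset sample providing its edge mask.
# Panned radial-coordinate textures form shards and plasma, using explicit UV derivatives for mip selection.
# Stage colors and a curve-remapped audio beat value are defined as reroutes, and summed layers feed a shared screen-detail function.
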Material /SparksMaps/Effects/Screens/Butterfly/M_Sparks_Screen_Butterfly
graph {
 properties {bAutomaticallySetUsageInEditor=false, bHasPixelAnimation=true, shadingModel=MSM_Unlit, usedShadingModels=Unlit}
 nodes [
  n1 = TextureCoordinate()
  DDX(Value=@n1)
  n2 = TextureCoordinate()
  DDX(Value=@n2)
  n3 = TextureCoordinate()
  DDX(Value=@n3)
  DDY(Value=@n1)
  DDY(Value=@n2)
  DDY(Value=@n3)
  n4 = Multiply(A=NamedRerouteUsage(), B=NamedRerouteUsage())
  n5 = Multiply(A=NamedRerouteUsage(), B=NamedRerouteUsage())
  n6 = Multiply(A=Multiply(A=Add(A=@n4, B=@n5), B=NamedRerouteUsage()), B=OneMinus(None=NamedRerouteUsage()))
  n7 = TextureCoordinate()
  n8 = Add(A=Constant2Vector() {g=0.5, r=0.5}, B=Multiply(A=NamedRerouteUsage()) {constB=-1})
  n9 = Saturate(None=OneMinus(None=SphereMask(A=@n7, B=@n8) {attenuationRadius=0.1, hardnessPercent=0}))
  n10 = OneMinus(None=SphereMask(A=TextureCoordinate(), B=@n8, Radius=ScalarParameter() {defaultValue=0.55, parameterName=Screen Edge Glow Radius}) {attenuationRadius=0.5, hardnessPercent=0})
  n11 = Multiply(
   A=Saturate(None=Power(Base=@n10) {constExponent=5})
   B=Multiply(A=NamedRerouteUsage(), B=ScalarParameter() {defaultValue=0.25, parameterName=Screen Edge Glow Intensity}) {constB=0.1}
  )
  n12 = OneMinus(None=NamedRerouteUsage())
  n13 = Power(Base=Multiply(A=NamedRerouteUsage(), B=Power(Base=@n12)))
  n14 = NamedRerouteUsage()
  n15 = Multiply(A=NamedRerouteUsage(), B=NamedRerouteUsage())
  n16 = NamedRerouteUsage()
  n17 = NamedRerouteUsage()
  n18 = NamedRerouteUsage()
  n19 = Add(A=@n18) {constB=0.1}
  n20 = MaterialFunctionCall(Vector or UVs=NamedRerouteUsage()) {materialFunction=&VectorToRadialValue}
  n21 = VectorParameter() {parameterName=Energy Tile and Pan, defaultValue=<rgba 3 0.15 0 -0.15>}
  n22 = AppendVector(A=@n21.R, B=@n21.R)
  n23 = TextureCoordinate()
  n24 = TextureSample!(
   UVs=Panner(Coordinate=Multiply(A=@n20.Radial Coordinates, B=@n22), Speed=AppendVector(A=@n21.B, B=@n21.B))
   Tex=NamedRerouteUsage()
   "DDX(UVs)"=DDX(Value=@n23)
   "DDY(UVs)"=DDY(Value=@n23)
   Apply View MipBias
  ) {
   mipValueMode=TMVM_Derivative
   samplerType=SAMPLERTYPE_LinearGrayscale
   texture=/Game/Effects/Fort_Effects/Emotes/UltraLight/Texture/T_Plasma_A_Greyscale
  }
  n25 = SmoothStep(Min=Constant() {r=0.3}, Max=Constant() {r=1}, Value=@n24.R)
  n26 = Multiply(A=SmoothStep(Min=Constant() {r=0.1}, Max=Constant() {r=0.5}, Value=@n24.R), B=Constant() {r=0.25}) {constB=0.2}
  n27 = Multiply(A=Power(Base=NamedRerouteUsage()), B=Add(A=Multiply(A=@n19, B=@n25), B=Multiply(A=@n18, B=@n26)))
  n28 = Add(A=NamedRerouteUsage(), B=Constant() {r=0.05}) {constB=0.1}
  n29 = Power(Base=NamedRerouteUsage(), Exp=Constant() {r=6}) {constExponent=5}
  n30 = VectorParameter() {parameterName=Energy Small Tile and Pan, defaultValue=<rgba 5 0 0 -0.15>}
  n31 = Multiply(
   A=MaterialFunctionCall(Vector or UVs=NamedRerouteUsage()) {materialFunction=&VectorToRadialValue}.Radial Coordinates
   B=AppendVector(A=@n30.R, B=@n30.R)
  )
  n32 = AppendVector(A=@n30.B, B=@n30.B)
  n33 = TextureCoordinate()
  n34 = TextureSample(
   UVs=Add(A=NamedRerouteUsage(), B=Panner(Coordinate=@n31, Speed=@n32))
   Tex=NamedRerouteUsage()
   "DDX(UVs)"=DDX(Value=@n33)
   "DDY(UVs)"=DDY(Value=@n33)
  ) {
   mipValueMode=TMVM_Derivative
   samplerType=SAMPLERTYPE_LinearGrayscale
   texture=&T_Plasma_A_Greyscale
  }
  n35 = Multiply(A=@n29, B=@n34.R)
  n36 = Constant() {r=3}
  n37 = Multiply(A=NamedRerouteUsage(), B=Add(A=@n29, B=Multiply(A=@n35, B=@n36)))
  n38 = Add(A=Add(A=Multiply(A=@n13, B=@n14), B=Multiply(A=@n15, B=@n16)), B=Add(A=Multiply(A=@n17, B=@n27), B=Multiply(A=@n28, B=@n37)))
  MaterialFunctionCall!(Emissive=Add(A=Multiply(A=@n6, B=@n9), B=Add(A=@n11, B=@n38)), Scanlines Tile) {materialFunction=/SparksMaps/Effects/Screens/Materials/MF/MF_Sparks_Screen_Details}
  n39 = Constant4Vector() {constant=<rgba 4 0.5 0 -0.1>}
  n40 = Panner(
   Coordinate=Multiply(
    A=MaterialFunctionCall() {materialFunction=&VectorToRadialValue}.Radial Coordinates
    B=AppendVector(A=@n39.R, B=@n39.R)
   )
   Speed=AppendVector(A=@n39.B, B=@n39.B)
  )
  n41 = TextureCoordinate()
  n42 = DDX(Value=@n41)
  n43 = DDY(Value=@n41)
  Multiply(
   A=Power(Base=TextureSample(UVs=@n40, "DDX(UVs)"=@n42, "DDY(UVs)"=@n43) {mipValueMode=TMVM_Derivative, texture=/Game/Effects/Fort_Effects/GameplayPulginsAssets/SPire/T_SandDustDotsSmall}.B) {constExponent=5}
   B=NamedRerouteUsage()
  )
  n44 = MaterialFunctionCall() {materialFunction=&VectorToRadialValue}
  n45 = Constant4Vector() {constant=<rgba 4 0.5 0 -0.1>}
  n46 = AppendVector(A=@n45.R, B=@n45.R)
  n47 = TextureSample!(UVs=Panner(Coordinate=Multiply(A=@n44.Radial Coordinates, B=@n46), Speed=AppendVector(A=@n45.B, B=@n45.B)), Tex, Apply View MipBias) {texture=/Game/Athena/Environments/Nexus/Textures/T_shard_bits_ZP}
  n48 = NamedRerouteUsage()
  n49 = NamedRerouteUsage()
  Multiply(A=Multiply(A=Multiply(A=@n47.G, B=@n48), B=Power(Base=@n49) {constExponent=1.5}), B=OneMinus(None=NamedRerouteUsage()))
  NamedRerouteDeclaration(None=TextureObject() {samplerType=SAMPLERTYPE_LinearGrayscale, texture=/Game/Effects/Fort_Effects/Textures/Tile/T_Caustic_Wind_Diagonal}) {name=EnergyTexture, nodeColor=<rgba 0 0 0 1>}
  n50 = Constant2Vector() {g=0.5, r=0.5}
  n51 = Multiply(A=NamedRerouteUsage()) {constB=-1}
  NamedRerouteDeclaration(None=SphereMask(A=TextureCoordinate(), B=Add(A=@n50, B=@n51), Radius=ScalarParameter() {defaultValue=0.5, parameterName=Center Sphere Mask Radius}) {attenuationRadius=0.5, hardnessPercent=0}) {name=Center Sphere Mask, nodeColor=<rgba 0 0 0 1>}
  n52 = NamedRerouteUsage()
  n53 = Constant() {r=3}
  NamedRerouteDeclaration(None=Multiply(A=NamedRerouteUsage(), B=Power(Base=@n52, Exp=@n53) {constExponent=4})) {name=ButterflyColor, nodeColor=<rgba 0 0 0 1>}
  n54 = MaterialFunctionCall(Vector or UVs=NamedRerouteUsage()) {materialFunction=&VectorToRadialValue}
  n55 = Constant4Vector() {constant=<rgba 3 1 0.01 -0.1>}
  n56 = AppendVector(A=@n55.R, B=@n55.R)
  n57 = TextureCoordinate()
  n58 = TextureSample!(
   UVs=Panner(Coordinate=Multiply(A=@n54.Radial Coordinates, B=@n56), Speed=AppendVector(A=@n55.B, B=@n55.B))
   Tex
   "DDX(UVs)"=DDX(Value=@n57)
   "DDY(UVs)"=DDY(Value=@n57)
   Apply View MipBias
  ) {mipValueMode=TMVM_Derivative, samplerType=SAMPLERTYPE_LinearColor, texture=&/T_Sparks_Screen_Butterfly}
  n59 = OneMinus(None=Add(A=@n58.R, B=@n58.R))
  n60 = MaterialFunctionCall(Vector or UVs=NamedRerouteUsage()) {materialFunction=&VectorToRadialValue}
  n61 = Constant4Vector() {constant=<rgba 10 3 0.1 -0.1>}
  n62 = AppendVector(A=@n61.R, B=@n61.R)
  n63 = TextureCoordinate()
  n64 = TextureSample(
   UVs=Panner(Coordinate=Multiply(A=@n60.Radial Coordinates, B=@n62), Speed=AppendVector(A=@n61.B, B=@n61.B))
   "DDX(UVs)"=DDX(Value=@n63)
   "DDY(UVs)"=DDY(Value=@n63)
  ) {mipValueMode=TMVM_Derivative, samplerType=SAMPLERTYPE_LinearColor, texture=&T_Sparks_Screen_Butterfly}
  NamedRerouteDeclaration(None=Add(A=@n58.R, B=Multiply(A=@n59, B=@n64.R))) {name=ShardBody, nodeColor=<rgba 0 0 0 1>}
  NamedRerouteDeclaration(None=Add(A=@n58.G, B=Multiply(A=@n59, B=@n64.G))) {name=ShardEdges, nodeColor=<rgba 0 0 0 1>}
  n65 = NamedRerouteUsage()
  n66 = NamedRerouteUsage()
  NamedRerouteDeclaration(None=Multiply(A=Add(A=@n65, B=@n66), B=@n9)) {name=Shard Mask, nodeColor=<rgba 0 0 0 1>}
  n67 = Multiply(A=Constant() {r=2}, B=NamedRerouteUsage())
  n68 = TextureCoordinate() {uTiling=2, vTiling=2}
  NamedRerouteDeclaration(None=Subtract(A=Add(A=@n67, B=@n68), B=Constant() {r=1})) {name=Radial Texture Offset, nodeColor=<rgba 0 0 0 1>}
  n69 = MaterialFunctionCall!(CustomUPosition) {materialFunction=/SparksStageShow/Systems/StageManager/FXManager/MaterialFunctions/MF_MojoStageColors}
  NamedRerouteDeclaration(None=StaticSwitchParameter(True=@n69.Secondary, False=@n69.Secondary) {parameterName=UseSecondary}) {name=Color, nodeColor=<rgba 0 0 0 1>}
  n70 = NamedRerouteUsage()
  n71 = NamedRerouteUsage()
  NamedRerouteDeclaration(None=Multiply(A=OneMinus(None=@n70), B=OneMinus(None=@n71))) {name=Overall Shard Mask Inverted, nodeColor=<rgba 0 0 0 1>}
  n72 = ScalarParameter() {parameterName=Horizontal Offset}
  n73 = Constant()
  n74 = Add(A=TextureCoordinate(), B=NamedRerouteDeclaration(None=AppendVector(A=@n72, B=@n73)) {name=Horizontal Offset, nodeColor=<rgba 0 0 0 1>})
  n75 = Add(A=Time(), B=ScalarParameter() {parameterName=TimeOffset})
  n76 = ConstantBiasScale(None=Sine(None=Multiply(A=@n75) {constB=0.5}))
  n77 = AppendVector(A=Add(A=Multiply(A=@n76) {constB=0.25}) {constB=0.4}, B=Constant() {r=0.65})
  n78 = ScalarParameter() {defaultValue=1, parameterName=ButterflyScale}
  n79 = TextureSample!(
   UVs=MaterialFunctionCall(UVs=@n74, Texture Scale=Multiply(A=@n77, B=@n78)) {materialFunction=&ScaleUVsByCenter}.UVs
   Tex
   Apply View MipBias
  ) {
   samplerSource=SSM_Clamp_WorldGroupSettings
   samplerType=SAMPLERTYPE_LinearColor
   texture=&T_Sparks_Screen_Butterfly
  }
  NamedRerouteDeclaration(None=@n79.B) {name=Butterfly Mask, nodeColor=<rgba 0 0 0 1>}
  n80 = VectorParameter() {parameterName=ColorA, defaultValue=<rgba 0.9427 0.0718 0 1>}
  NamedRerouteDeclaration(None=@n80.RGB) {name=PrimaryColor, nodeColor=<rgba 0 0 0 1>}
  NamedRerouteDeclaration(None=@n80.RGB) {name=SecondaryColor, nodeColor=<rgba 0 0 0 1>}
  n81 = Add(A=AppendVector(A=LinearInterpolate(Alpha=@n76) {constA=0.02, constB=0.005}, B=Constant()), B=@n77)
  n82 = TextureSample(UVs=MaterialFunctionCall(UVs=@n74, Texture Scale=Multiply(A=@n81, B=@n78)) {materialFunction=&ScaleUVsByCenter}.UVs) {
   samplerSource=SSM_Clamp_WorldGroupSettings
   samplerType=SAMPLERTYPE_LinearColor
   texture=&T_Sparks_Screen_Butterfly
  }
  NamedRerouteDeclaration(None=@n82.B) {name=Butterfly Edge Offset, nodeColor=<rgba 0 0 0 1>}
  n83 = CurveAtlasRowParameter(CurveTime=CollectionParameter() {collection=/FMAudioReactiveData/AudioAnalysis/MPC_Sparks_Audio_Parameters, parameterName=BeatLifetime}) {
   atlas=/SparksMaps/Effects/Screens/Curves/CurveAtlas_Audio_Beat_A
   curve=/SparksMaps/Effects/Screens/Curves/Curve_Audio_Beat_Remap_001
   parameterName=ReactivityCurve
  }
  NamedRerouteDeclaration(None=MaterialFunctionCall(Input=@n83, Input Low=Constant(), Input High=Constant() {r=1}, Target Low=Constant() {r=1}, Target High=Constant() {r=2}) {materialFunction=&RemapValueRange}.Result) {name=Reactivity, nodeColor=<rgba 0 0 0 1>}
  n84 = TextureSample(UVs=Panner(Coordinate=TextureCoordinate()) {speedX=0.25, speedY=0.25}) {samplerType=SAMPLERTYPE_Grayscale, texture=&LowResBlurredNoise}
  NamedRerouteDeclaration(None=Saturate(None=Add(A=@n84.R) {constB=0.1})) {name=Metallic Sheen, nodeColor=<rgba 0 0 0 1>}
  VectorParameter() {parameterName=ColorB, defaultValue=<rgba 0 0.3538 0.9375 1>}
 ]
}
